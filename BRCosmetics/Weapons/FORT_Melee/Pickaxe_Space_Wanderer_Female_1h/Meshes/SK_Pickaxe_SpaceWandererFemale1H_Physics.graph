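# Physics asset for the one-handed Space Wanderer Female pickaxe cosmetic, covering its dynamic cable bones.
# The dyn_master_cable body combines a tiny capsule with a very large, offset and tilted capsule, while dyn_cable has one small capsule.
# A single constraint joins dyn_cable to dyn_master_cable with swing and twist angle limits.
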PhysicsAsset /BRCosmetics/Weapons/FORT_Melee/Pickaxe_Space_Wanderer_Female_1h/Meshes/SK_Pickaxe_SpaceWandererFemale1H_Physics
bodies {
 dyn_cable [{shapeType=Capsule, length=3, radius=1, center=<3.8329 0.3333 0>}]
 dyn_master_cable [
  {shapeType=Capsule, length=0.5, radius=0.5}
  {shapeType=Capsule, length=15, radius=36.7534, center=<1.8257 0.001 -20.2201>, rotation=<rot 54.9996 0 0>}
 ]
}
constraints [{bone1Name=dyn_cable, bone2Name=dyn_master_cable, swing1LimitDegrees=45, swing2LimitDegrees=45, twistLimitDegrees=30}]
