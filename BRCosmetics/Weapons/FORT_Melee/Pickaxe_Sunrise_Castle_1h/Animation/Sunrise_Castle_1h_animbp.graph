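# Animation Blueprint for the Sunrise Castle one-handed pickaxe that adds secondary physics motion to the weapon mesh.
# The input pose passes through DefaultSlot into a RigidBody node simulating relative to b_AR_Root, using a physics asset borrowed from the Space Wanderer Female pickaxe.
# A Closed-to-Into state machine and a standalone Closed sequence player exist but are not wired to the output.
# The event graph holds only stub nodes with no logic.
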AnimBlueprint /BRCosmetics/Weapons/FORT_Melee/Pickaxe_Sunrise_Castle_1h/Animation/Sunrise_Castle_1h_animbp
target_skeleton=/BRCosmetics/Weapons/FORT_Melee/Pickaxe_Sunrise_Castle_1h/Meshes/Sunrise_Castle_1h_Skeleton
graphs [
 AnimGraph {nodes [
   n1 = Input Pose()
   n2 = RigidBody(ComponentPose=Local To Component(LocalPose=Slot 'DefaultSlot'(Source=@n1))) {
    bClampLinearTranslationLimitToRefPose=true
    bFreezeIncomingPoseOnStart=true
    bOverrideWorldGravity=true
    overridePhysicsAsset=/BRCosmetics/Weapons/FORT_Melee/Pickaxe_Space_Wanderer_Female_1h/Meshes/SK_Pickaxe_SpaceWandererFemale1H_Physics
    simulationSpace=BaseBoneSpace
    alphaBoolBlend {blendOption=Linear}
    baseBoneRef {boneName=b_AR_Root}
    componentLinearAccScale=<0.5 0.3 0.3>
    componentLinearVelScale=<0.2 0.2 0.2>
    overrideWorldGravity {z=-980}
   }
   Output Pose(Result=Component To Local(ComponentPose=@n2)) {layerGroup=None}
   Sequence Player 'Sunrise_Castle_1h_Closed_PROP'() {bLoopAnimation=false, sequence=&/Sunrise_Castle_1h_Closed_PROP}
   AnimGraphNode_StateMachine_0 = New State Machine() {is=title AnimGraphNode_StateMachine}
  ]}
 New State Machine {
  entry=State
  path=AnimGraph.AnimGraphNode_StateMachine_0.New State Machine
  ids=id
  nodes [
   AnimStateNode_0 = State() {is=title AnimStateNode}
   AnimStateNode_1 = State_1() {is=title AnimStateNode}
   AnimStateTransitionNode_0 = AnimStateTransitionNode() {bAutomaticRuleBasedOnSequencePlayerInState=true}
  ]
  links [{
    dst=State_1
    rule=AnimGraph.AnimGraphNode_StateMachine_0.New State Machine.AnimStateTransitionNode_0.Transition
    src=State
    via=AnimStateTransitionNode_0
   }]
 }
 State {
  path=AnimGraph.AnimGraphNode_StateMachine_0.New State Machine.AnimStateNode_0.State
  nodes [Output Animation Pose(Result=Sequence Player 'Sunrise_Castle_1h_Closed_PROP'() {bLoopAnimation=false, sequence=&Sunrise_Castle_1h_Closed_PROP})]
 }
 State_1 {
  path=AnimGraph.AnimGraphNode_StateMachine_0.New State Machine.AnimStateNode_1.State_1
  nodes [Output Animation Pose(Result=Sequence Player 'Sunrise_Castle_1h_Into_PROP'() {sequence=&/Sunrise_Castle_1h_Into_PROP})]
 }
 Transition {path=AnimGraph.AnimGraphNode_StateMachine_0.New State Machine.AnimStateTransitionNode_0.Transition, nodes [Result()]}
 EventGraph {nodes [
   TryGetPawnOwner()
   Event BlueprintUpdateAnimation()
  ]}
]
variables [
 offhand_weapon {type=boolean}
 Skel_Mesh {type=SkeletalMeshComponent}
]
